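# Post-process Animation Blueprint for the YogaPatio male head that damps head rotation by partially counter-rotating the head bone.
# Each update reads the head's delta from reference pose relative to neck_02, then negates and scales each axis by its percentage variable.
# The AnimGraph applies this additively in bone space.
# Its alpha is on with a valid pawn owner, and off without one, in the front end or during emotes.
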AnimBlueprint /BRCosmetics/Characters/Player/Male/Medium/Heads/M_MED_YogaPatio_Head/Meshes/M_MED_YogaPatio_Head_PP_AnimBP
target_skeleton=/Game/Characters/Player/Common/Fortnite_Base_Head/Fortnite_Base_Head_Export_Skeleton
graphs [
 AnimGraph {nodes [
   n1 = "Transform (Modify) Bone - Bone: head"(
    ComponentPose=Local To Component(LocalPose=Input Pose())
    Alpha=Get enableHeadRotLimitAlpha() -> (enableHeadRotLimitAlpha:double)
    Translation="0,0,0"
    Rotation=Get Head Rot() -> (Head Rot:Rotator)
    Scale="1,1,1"
   ) {rotationMode=BMM_Additive, rotationSpace=BCS_BoneSpace, alphaBoolBlend {blendOption=Linear}, boneToModify {boneName=head}}
   Output Pose(Result=Component To Local(ComponentPose=@n1))
  ]}
 EventGraph {nodes [
   Event BlueprintUpdateAnimation()
   > n1 = Sequence!(execute) -> (then_0, then_1, then_2:Exec)
   > n2 = Is Valid(InputObject=TryGetPawnOwner())
   Set enableHeadRotLimitAlpha(execute=@n2.Is Not Valid, enableHeadRotLimitAlpha:double=0) -> (Output_Get:double)
   Set enableHeadRotLimitAlpha(execute=@n2.Is Valid, enableHeadRotLimitAlpha=1)
   Branch(execute=@n1.then_1, Condition=OR Boolean(A=Get bIsInFrontEnd(), B=Get bIsPlayingEmote()))
   > Set enableHeadRotLimitAlpha(enableHeadRotLimitAlpha=0)
   n3 = GetOwnerSkeletalMeshComponent()
   n4 = BreakRotator(InRot=BreakTransform(InTransform=GetDeltaTransformFromRefPose(self=@n3, BoneName=head, BaseName=neck_02)).Rotation)
   n5 = Get Z Percentage() -> (Z Percentage:double)
   n6 = Get X Percentage() -> (X Percentage:double)
   n7 = Get Y Percentage() -> (Y Percentage:double)
   Set Head Rot(
    execute=@n1.then_2
    Head Rot:Rotator=MakeRotator(
     Roll=float * float!(A=@n4, B=-1, C:double=@n6) -> (ReturnValue) {is=title K2Node_PromotableOperator}
     Pitch=float * float(A=@n4.Pitch, B=-1, C=@n7)
     Yaw=float * float(A=@n4.Yaw, B=-1, C=@n5)
    )
   ) -> (Output_Get:Rotator)
   TryGetPawnOwner()
  ]}
]
variables [
 enableHeadRotLimitAlpha {type=number}
 Head Rot {type=Rotator}
 X Percentage {type=number, default=0.3}
 Y Percentage {type=number, default=0.3}
 Z Percentage {type=number, default=0.7}
]
